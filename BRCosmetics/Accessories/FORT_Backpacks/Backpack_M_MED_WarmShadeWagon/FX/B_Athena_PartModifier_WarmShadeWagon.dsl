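; Cosmetic part modifier for WarmShadeWagon that sets up reactive idle VFX and materials on the backpack or body part.
; On init it detects part type and frontend state, then sets Niagara reactivity from the Emissive variant.
; In the frontend, a delayed back-and-forth timeline fakes reactivity through an audio-curve variable, a Niagara debug flag and dynamic material scalars.
; Backpack materials follow the curve smoothly, while other parts snap on or off.
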(event EventOnPostResIn
  (bind _returnvalue (CallFunction|GetIdleVFXComponent))
  (|CallParentFunction)
  (Utilities|FlowControl|Sequence
    (:then_0
      (CallFunction|SetBodyPartBool)
      (Variables|Default|SetIsFrontend (CustomCosmeticModifier|IsinFrontEnd)))
    (:then_1
      (if (Variables|Default|GetIsFrontend)
        ; Frontend Reactivity Loop
        (Utilities|FlowControl|Delay 5.0)
        (bind (_direction _newtrack_0) (Timeline|Timeline PlayFromStart)
          (:Update
            (CallFunction|SetMaterialValues _newtrack_0)
            (Variables|Default|SetFakeAudioCurve _newtrack_0)
            (if (<= _newtrack_0 0.01)
              _
              (elif (>= _newtrack_0 0.1)
                (bind _asniagara_particle_system_component (Utilities|Casting|CastToNiagaraComponent _returnvalue :!label _label))
                (Effects|Components|ParticleSystem|SetBoolParameter _asniagara_particle_system_component "ReactiveDebug" true))))
          (:Finished
            (Utilities|FlowControl|Sequence
              (:then_0
                (if (not (<= _newtrack_0 0.01))
                  (goto _label)))
              (:then_1
                (Utilities|FlowControl|Delay 5.0)
                (if (<= _newtrack_0 0.01)
                  (Timeline|Timeline PlayFromStart)
                  (else
                    (Timeline|Timeline ReverseFromEnd)))))))
        (bind _asniagara_particle_system_component_2 (Utilities|Casting|CastToNiagaraComponent _returnvalue))
        (Effects|Components|ParticleSystem|SetBoolParameter _asniagara_particle_system_component_2 "ReactiveDebug" false)))
    (:then_2
      ;[0.3024 0.1295 0.6719 0.5] FX - Reactivity check
      (bind _asfort_player_pawn (Utilities|Casting|CastToFortPlayerPawn (Actor|GetOwner)))
      (if (Variables|Default|GetIsBackpack)
        (Variables|Default|SetSlotTag "(TagName=\"LoadoutSlot.Backpack\")")
        (else
          (Variables|Default|SetSlotTag "(TagName=\"LoadoutSlot.Character\")")))
      (bind _index (CosmeticLoadout|GetVariantChannelActiveIndexforSlotActor (Variables|Default|GetSlotTag) _asfort_player_pawn "(TagName=\"Cosmetics.Variant.Channel.Emissive\")" (Utilities|Array|MakeArray "(TagName=\"Cosmetics.Variant.Property.Emissive0\")" "(TagName=\"Cosmetics.Variant.Property.Emissive1\")") false false))
      (bind _output_get (Variables|Default|SetbUseFXReactivity (Utilities|Select :Index _index :Option%200 true :Option%201 false)))
      (bind _asniagara_particle_system_component_3 (Utilities|Casting|CastToNiagaraComponent (CallFunction|GetIdleVFXComponent)))
      (Niagara|SetNiagaraVariable(Bool) _asniagara_particle_system_component_3 "bUseReactivity" _output_get))))

(event Custom|SetMaterialValues
  (params (Value :type Float%20%28double-precision%29))
  (bind _returnvalue (CustomCharacter|GetSkeletalMeshForPartOnPlayerPawn))
  (for (_ _array_index) (Rendering|Material|GetMaterials _returnvalue)
    (Rendering|Material|SetScalarParameterValue (Class|BTechArtGlobalLibrary|GetDynamicMaterialInstance _returnvalue _array_index) "ReactivityDebug" (Math|Float|SelectFloat Value (Math|Float|SelectFloat 1.0 0.0 (> Value 0.01)) (Variables|Default|GetIsBackpack)) :!class /Script/Engine.MaterialInstanceDynamic)))

(fn SetBodyPartBool
  (switch Utilities|FlowControl|Switch|SwitchonEFortCustomPartType (CustomCharacter|GetCharacterPartType)
    (:Body
      (Effects|Components|ParticleSystem|SetBoolParameter (CallFunction|GetIdleVFXComponent) "isBody" true))
    (:Backpack
      (Variables|Default|SetIsBackpack true))))

(fn ConstructionScript
  (|CallParentFunction))
