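; Character part modifier for the WarmShadeWagon backpack that makes the part's material react to music.
; On post-res-in it reads the material's ReactivityOn parameter to enable the effect.
; Each tick it adds scaled music amplitude from the audio-reactive parameter collection to a clamped, decaying accumulator driving AccumulatedSound.
; In the frontend, it substitutes a fake audio curve value when amplitude and accumulator are not above zero.
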(event EventOnPostResIn
  (|CallParentFunction)
  (Utilities|FlowControl|Sequence
    (:then_0
      (bind _returnvalue (Rendering|Material|GetScalarParameterValue (Class|BTechArtGlobalLibrary|GetDynamicMaterialInstance (CustomCharacter|GetSkeletalMeshForPartOnPlayerPawn)) "ReactivityOn" :!class /Script/Engine.MaterialInstanceDynamic))
      (Variables|Default|SetReactivityOn (> _returnvalue 0)))))

(event EventTick
  (params (DeltaSeconds :type Float%20%28single-precision%29))
  (bind _returnvalue (CustomCharacter|GetSkeletalMeshForPartOnPlayerPawn))
  (bind _material_instance_dynamic (Class|BTechArtGlobalLibrary|GetDynamicMaterialInstance _returnvalue))
  (|CallParentFunction DeltaSeconds)
  (if (Variables|Default|GetReactivityOn)
    (bind _returnvalue_2 (Rendering|Material|GetScalarParameterValue "/Game/Athena/Events/AudioAnalysis/MPC_Audio_Reactive_Parameters.MPC_Audio_Reactive_Parameters" "Amplitude_Music" :!class /Script/Engine.KismetMaterialLibrary))
    (bind _output_get (Variables|Default|SetAudioAmplitude _returnvalue_2))
    (bind _output_get_2 (Variables|Default|SetLastAccum (Math|Interpolation|FInterpTo (Math|Float|Clamp(Float) (+ (* _output_get 0.02) (Variables|Default|GetLastAccum))) 0.0 DeltaSeconds 0.7)))
    (if (Variables|Default|GetIsFrontend)
      (if (or (> (Variables|Default|GetAudioAmplitude) 0) (> (Variables|Default|GetLastAccum) 0))
        (Rendering|Material|SetScalarParameterValue _material_instance_dynamic "AccumulatedSound" (Variables|Default|GetLastAccum) :!class /Script/Engine.MaterialInstanceDynamic)
        (else
          (Rendering|Material|SetScalarParameterValue _material_instance_dynamic "AccumulatedSound" (Variables|Default|GetFakeAudioCurve) :!class /Script/Engine.MaterialInstanceDynamic)))
      (else
        (Rendering|Material|SetScalarParameterValue _material_instance_dynamic "AccumulatedSound" _output_get_2 :!class /Script/Engine.MaterialInstanceDynamic)))))

(fn ConstructionScript
  (|CallParentFunction))
