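# Unlit translucent lightning bolt material for Niagara mesh particles on the WarmShadeWagon backpack.
# Vertex math widens the mesh perpendicular to its Y axis and the camera, squeezes tips and view-aligned widths, and blends in a normal-based, audio-reactive offset.
# Per-particle random noise carves soft-edged width opacity, while gradient masks from dynamic parameters drive a traveling reveal and brightness boost over particle color.
# An anisotropic noise branch is marked as currently unused.
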Material /BRCosmetics/Accessories/FORT_Backpacks/Backpack_M_MED_WarmShadeWagon/FX/M_WarmShadeWagon_LightningBolt_Splines
graph {
 properties {
  bUsedWithNiagaraMeshParticles=true
  blendMode=BLEND_Translucent
  pixelDepthOffsetMode=PDOM_Legacy
  shadingModel=MSM_Unlit
  twoSided=true
  usedShadingModels=Unlit
 }
 nodes [
  n1 = Normalize(VectorInput=Subtract(A=CameraPositionWS(), B=WorldPosition().XYZ))
  n2 = Normalize(VectorInput=Transform(None=Constant3Vector() {constant=<rgba 0 1 0 0>}))
  n3 = Normalize(VectorInput=CrossProduct(A=@n1, B=@n2))
  n4 = TextureCoordinate()
  n5 = DynamicParameter() {
   parameterIndex=1
   defaultValue=<rgba 1 1 1 0>
   paramNames [Width Noise Contribution, Mesh Width, Anisotropic Noise Contribution, Audio Reactivity Intensity]
  }
  n6 = Multiply(A=Subtract(A=ComponentMask(None=@n4) {r=true}) {constB=0.5}, B=@n5.Mesh Width)
  n7 = ComponentMask(None=@n4) {g=true}
  n8 = OneMinus(None=Abs(None=ConstantBiasScale(None=@n7) {bias=-0.5, scale=2}))
  n9 = Power(Base=@n8, Exp=ScalarParameter() {defaultValue=0.5, parameterName=Tip Squeeze Exponent})
  n10 = ScalarParameter() {defaultValue=1, parameterName=Tip Squeeze Amount}
  n11 = OneMinus(None=Multiply(A=OneMinus(None=@n9), B=Saturate(None=@n10)))
  n12 = DotProduct(A=@n2, B=@n1)
  n13 = Power(Base=OneMinus(None=Abs(None=@n12)), Exp=ScalarParameter() {defaultValue=0.5, parameterName=Width View Angle Squish Exponent})
  n14 = ScalarParameter() {defaultValue=0.5, parameterName=Width View Angle Squish Amount}
  n15 = OneMinus(None=Multiply(A=OneMinus(None=@n13), B=Saturate(None=@n14)))
  n16 = LinearInterpolate(Alpha=Saturate(None=Power(Base=@n8) {constExponent=5})) {constA=-0.5}
  n17 = Constant3Vector() {constant=<rgba 1 1 -1 0>}
  n18 = Multiply(A=Multiply(A=VertexNormalWS(), B=LinearInterpolate(B=@n17) {constA=1, constAlpha=0.3})) {constB=10}
  n19 = DynamicParameter() {
   parameterIndex=1
   defaultValue=<rgba 1 1 1 0>
   paramNames [Width Noise Contribution, Mesh Width, Anisotropic Noise Contribution, Audio Reactivity Intensity]
  }
  Add(
   A=Multiply(A=Multiply(A=@n3, B=@n6), B=Multiply(A=@n11, B=@n15) {constA=-208.9883})
   B=LinearInterpolate(B=Multiply(A=@n16, B=@n18), Alpha=@n19.Audio Reactivity Intensity)
  )
  n20 = ParticleColor()
  n21 = NamedRerouteUsage()
  MaterialFunctionCall(Emissive=LinearInterpolate(A=@n20.RGB, B=Multiply(A=@n20.RGB, B=@n21), Alpha=NamedRerouteUsage())) {materialFunction=&MF_FX_EmissiveAdjustments}
  Multiply(A=NamedRerouteUsage(), B=@n20.A)
  Multiply(A=@n18, B=LinearInterpolate(Alpha=@n19.Audio Reactivity Intensity) {constA=0.5, constB=1.5})
  n22 = TextureCoordinate()
  n23 = DynamicParameter() {paramNames [Lightning Reveal Position, Lightning Reveal Size, Lightning Brightness Boost Size, Lightning Brightness Boost Intensity]}
  n24 = MaterialFunctionCall!(
   Gradient=OneMinus(None=ComponentMask(None=@n22) {g=true})
   Mask Position=@n23.Lightning Reveal Position
   Mask Size=@n23.Lightning Reveal Position
  ) {materialFunction=/Game/Effects/Fort_Effects/MaterialFunctions/MF_GradientMask}
  n25 = OneMinus(None=Abs(None=ConstantBiasScale(None=@n24.Linear) {bias=-0.5, scale=2}))
  n26 = ScalarParameter() {defaultValue=0.16, parameterName=Lightning Thickness Squeeze Size}
  NamedRerouteDeclaration(None=Saturate(None=Divide(A=@n25, B=@n26))) {name=Lightning Thickness Multiplier, nodeColor=<rgba 1 0.1875 0 1>}
  n27 = TextureCoordinate()
  n28 = Multiply(A=ComponentMask(None=@n27) {g=true}, B=ScalarParameter() {defaultValue=2, parameterName=Aniso Noise Scale})
  n29 = Multiply(A=ComponentMask(None=@n27) {r=true}, B=ScalarParameter() {defaultValue=0.2, parameterName=Aniso Noise Skew})
  n30 = TextureSample(UVs=AppendVector(A=ParticleRandom(), B=Add(A=@n28, B=@n29))) {samplerType=SAMPLERTYPE_LinearGrayscale, texture=/Game/Animation/Game/MainPlayer/Emotes/Congestion/FX/T_Noise_Anisotropic}
  n31 = ScalarParameter() {defaultValue=1, parameterName=Aniso Exponent}
  n32 = DynamicParameter() {
   parameterIndex=1
   defaultValue=<rgba 1 1 1 0>
   paramNames [Width Noise Contribution, Mesh Width, Anisotropic Noise Contribution, Audio Reactivity Intensity]
  }
  NamedRerouteDeclaration(None=Multiply(A=Power(Base=@n30.R, Exp=@n31), B=@n32.Anisotropic Noise Contribution)) {name=Anisotropic Noise, desc=Currently unused... maybe delete?, nodeColor=<rgba 0.0547 1 0 1>}
  NamedRerouteDeclaration(None=Power(Base=MaterialFunctionCall(Gradient=@n24.Linear, Mask Position=@n23.Lightning Reveal Position, Mask Size=@n23.Lightning Reveal Position) {materialFunction=&MF_GradientMask}.Sinusoidal) {constExponent=6}) {name=Lightning Brightness Boost Mask, nodeColor=<rgba 1 0 0.9609 1>}
  NamedRerouteDeclaration(None=@n23.Lightning Brightness Boost Intensity) {name=Lightning Brightness Boost Intensity, nodeColor=<rgba 0.9922 0 1 1>}
  n33 = ConstantBiasScale(None=ComponentMask(None=TextureCoordinate()) {r=true}) {bias=-0.5, scale=2}
  n34 = ComponentMask(None=TextureCoordinate()) {g=true}
  n35 = ScalarParameter() {defaultValue=0.75, parameterName=Width Noise V Scale}
  n36 = TextureSample(UVs=AppendVector(A=ParticleRandom(), B=Multiply(A=@n34, B=@n35))) {
   samplerType=SAMPLERTYPE_LinearGrayscale
   texture=/BRCosmetics/Weapons/FORT_Melee/Pickaxe_Parfait_Female/FX/Textures/T_TrifleParfait_LightningNoise
  }
  n37 = ScalarParameter() {defaultValue=1, parameterName=Width Noise Exponent}
  n38 = DynamicParameter() {
   parameterIndex=1
   defaultValue=<rgba 1 1 1 0>
   paramNames [Width Noise Contribution, Mesh Width, Anisotropic Noise Contribution, Audio Reactivity Intensity]
  }
  n39 = Subtract(A=OneMinus(None=Abs(None=@n33)), B=Multiply(A=Power(Base=@n36.R, Exp=@n37), B=@n38.Width Noise Contribution))
  n40 = OneMinus(None=NamedRerouteUsage())
  n41 = Divide(A=Subtract(A=Subtract(A=@n39, B=@n40), B=NamedRerouteUsage()), B=ScalarParameter() {defaultValue=0.2, parameterName=Width Edge Softness})
  NamedRerouteDeclaration(None=Power(Base=Saturate(None=@n41), Exp=ScalarParameter() {defaultValue=1, parameterName=Width Opacity Exponent})) {name=Width Opacity, nodeColor=<rgba 0.3828 0 1 1>}
  ObjectOrientation()
  ParticleRelativeTime()
 ]
}
