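# Niagara system for the WarmShadeWagon cosmetic backpack, producing music-reactive FX: speaker rings, top glow and lightning on the pack, plus lightning on the wearer's body.
# System.bIsBackpack and User.isBody select which emitters spawn, with most also gated by bUseReactivity.
# System.MusicReactive, audio amplitude and timeline drive spawn rates, scale, color and lifetime.
# Body lightning follows skeletal-mesh bones; lightning emitters derive layered glow colors and material parameters for multiple mesh renderers.
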NiagaraSystem /BRCosmetics/Accessories/FORT_Backpacks/Backpack_M_MED_WarmShadeWagon/FX/NS_WarmShadeWagon_Splines
dependencies {
 usedDataInterfaces [
  &NiagaraDataInterfaceCurve
  &NiagaraDataInterfaceArrayFloat3
  &NiagaraDataInterfaceArrayFloat
  &NiagaraDataInterfaceColorCurve
  &NiagaraDataInterfaceSkeletalMesh
 ]
 usedDynamicInputs [&Boolean_OrOperation, &SetBoolByFloatComparison]
 usedModules [
  &self:SystemSpawnScript.SystemScriptSource.SystemScriptGraph.NiagaraNodeAssignment_2.SetVariables_#1
  &self:SystemSpawnScript.SystemScriptSource.SystemScriptGraph.NiagaraNodeAssignment_3.SetVariables_#2
  &SystemState
  &Timeline
  &self:SystemSpawnScript.SystemScriptSource.SystemScriptGraph.NiagaraNodeAssignment_1.SetVariables_#3
  &self:SystemSpawnScript.SystemScriptSource.SystemScriptGraph.NiagaraNodeAssignment_0.SetVariables_#4
 ]
}
emitters {
 Backpack_SpeakerRings {
  emitterUpdateScript [
   &EmitterState(
    Inactive Response=::NewEnumerator1
    Life Cycle Mode=::NewEnumerator1
    Loop Behavior=::NewEnumerator1
    Loop Duration Mode=::NewEnumerator1
    UseLoopDelay=false
   )
   &SpawnBurst_Instantaneous(
    Spawn Count=2
    Spawn Probability=&MakeCustomFloatFromBool(Bool=$System.bIsBackpack, True Float=1, False Float=0)
    Use Spawn Probability=true
    UseLoopCountLimit=false
   )
  ]
  particleSpawnScript [
   &PartitionParticles(ID to Partition:NiagaraInt32=&ReturnExecIndex())
   &InitializeParticle(
    Mesh Scale Mode=::NewEnumerator3
    Position Offset=&SelectVectorFromArray(
     Direct Array Index=$Particles.PartitionParticles.Partition
     Vector Selection Array=DataInterface() {propertyValues {FloatData [<11 -7.6 -0.55>, <-11 -7.6 -0.55>]}}
    )
    Sprite Size Mode=::NewEnumerator3
    Uniform Sprite Size=27
    UsePositionOffset=true
    Write Material Random=true
   )
   &InitialMeshOrientation(Mesh Orientation Mode=::NewEnumerator1, Rotation=<0.25 0 0>, Use Rotation Vector=true)
   &CameraOffset(Camera Offset Amount=&FN_FrontEndFloatModifier(Gameplay=0, Front End=0))
  ]
  particleUpdateScript [
   &ParticleState(Kill Particles When Lifetime Has Elapsed=false, Let Infinitely Lived Particles Die When Emitter Deactivates=true)
   &UpdateMeshOrientation(
    Rotation Rate=&SelectFloatFromArray(Direct Array Index=$Particles.PartitionParticles.Partition, Float Selection Array=DataInterface() {propertyValues {FloatData [-0.5,0.5]}})
    Rotation Vector=<0 1 0>
   )
   &ScaleColor(
    Scale Alpha=&FloatFromCurve(FloatCurve=DataInterface() {propertyValues {Curve {keys [<at 0.3 0>,<at 0.5 1>]}}}, CurveIndex=$System.MusicReactive)
    Scale RGB=&Multiply_VectorByFloat(
     Vector=&Multiply_Vector(
      A=<1 1 1>
      B=&Lerp_Vector(
       A=<1 1 1>
       B=<50 1 1>
       Alpha=&Power(a=&Multiply_Float(A=$System.MusicReactive, B=$NPC.FortniteNPC.EventAudio_Amplitude), b=15)
      )
     )
     Float=2
    )
    ScaleA=true
    ScaleRGB=true
   )
   &ScaleMeshSize(Scale Factor=&Multiply_VectorByFloat(
     Vector=<1 1 1>
     Float=&Lerp_Float(A=0.2, B=0.3, Alpha=&Power(a=$System.MusicReactive, b=2))
    ))
   &FN_Mesh_Attributes(
    Create Color Attribute=true
    Create Scale Attribute=true
    Scale=&Multiply_VectorByFloat(
     Vector=<1 1 1>
     Float=&Multiply_Float(A=0.56, B=&Lerp_Float(A=0.5, B=1, Alpha=&Power(a=$System.MusicReactive, b=2)))
    )
    Write to Module Namespace=true
   ) {enabled=false}
   &FN_Mesh_Attributes(
    Color=&MakeLinearColorFromVectorAndFloat(
     "Vector (RGB)"=&Multiply_Vector(
      A=&MakeVectorFromLinearColorRGB(Color=&MultiplyLinearColor(Linear Color=<rgba 1 0.2285 0.2584 1>, Scale Factor=&MakeVector4_FromFloat(Float=5)))
      B=&Lerp_Vector(
       A=<0.5 0.5 0.5>
       B=<5 5 1>
       Alpha=&Power(a=&Multiply_Float(A=$System.MusicReactive, B=$NPC.FortniteNPC.EventAudio_Amplitude), b=10)
      )
     )
     "Float (Alpha)"=&FloatFromCurve(FloatCurve=DataInterface() {propertyValues {Curve {keys [<at 0.3 0>,<at 0.5 1>]}}}, CurveIndex=$System.MusicReactive)
    )
    Create Color Attribute=true
    Create Scale Attribute=true
    Scale=&Multiply_VectorByFloat(
     Vector=<1 1 2>
     Float=&Multiply_Float(A=0.5, B=&Lerp_Float(A=0.5, B=0.8, Alpha=&Power(a=$System.MusicReactive, b=2)))
    )
    Write to Module Namespace=true
   ) {module=FN_Mesh_Attributes001}
   /SetVariables(Particles.Position002:NiagaraPosition=&AddVectorToPosition(Position=$Particles.Position, Vector=<0 1 0>)) {isSetParameters=true}
   /SetVariables(Particles.MeshOrientation02:Quat4f=/Niagara/DynamicInputs/Rotation/InvertQuaternion(Quat:Quat4f=$Particles.MeshOrientation)) {isSetParameters=true}
   &KillParticles(Kill Particles=$System.BIsBody)
  ]
  rendererClasses [&NiagaraMeshRendererProperties]
  renderers [
   {rendererClass=&NiagaraMeshRendererProperties, rendererIndex=0}
   {rendererClass=&NiagaraMeshRendererProperties, rendererIndex=1}
   {rendererClass=&NiagaraMeshRendererProperties, rendererIndex=2}
  ]
 }
 Backpack_TopGlow {
  emitterUpdateScript [
   &EmitterState(Inactive Response=::NewEnumerator1, Life Cycle Mode=::NewEnumerator1, Loop Duration=1, UseLoopDelay=false)
   &SpawnRate(
    Spawn Probability=&MakeCustomFloatFromBool(Bool=&Boolean_AndOperation(Bool1=$System.bIsBackpack, Bool2=$User.bUseReactivity), True Float=1, False Float=0)
    SpawnRate=&Multiply_Float(A=20, B=$System.MusicReactive)
    Use Spawn Probability=true
   )
  ]
  particleSpawnScript [
   &InitializeParticle(
    Color=&MakeCustomLinearColorFromBool(Bool=$System.bUseRedderColors, True Color=<rgba 6 0.006 0.006 1>, False Color=<rgba 6 0 0.1 1>)
    Mesh Scale=&Multiply_VectorByFloat(Vector=&RandomRangeVector(Minimum=<2 2 0.6>, Maximum=<3 3 0.8>), Float=0.2)
    Mesh Scale Mode=::NewEnumerator4
    Position Offset=<10 0 30>
    Sprite Size Mode=::NewEnumerator3
    UsePositionOffset=false
    Write Material Random=true
   )
   &SystemLocation(MaskSpawnGroup=false, Offset=&RandomRangeVector(Minimum=<20 -3 10>, Maximum=<-20 -2 10>))
   /SetVariables(Particles.Random=&RandomRangeFloat(Minimum=0, Maximum=1)) {isSetParameters=true}
   &InitialMeshOrientation(Mesh Orientation Mode=::NewEnumerator2, Rotation=<0 0 0.1>, Use Rotation Vector=true)
   &CameraOffset()
  ]
  particleUpdateScript [
   &ParticleState()
   &self:SquareMovement(Disapear=0.1, Fade=170, Offset=85) {module=SquareMovement}
   &SolveForcesAndVelocity(Clamp Velocity=false, Limit Acceleration=false)
   &ScaleColor(
    Scale Alpha=2
    Scale RGB=&VectorFromFloat(Value=&Lerp_Float(A=0.5, B=1.5, Alpha=&Power(a=$NPC.FortniteNPC.EventAudio_Amplitude, b=4)))
    ScaleA=true
    ScaleRGB=true
   )
   &ScaleMeshSize(Scale Factor=&MakeVector(
     X=&FN_FrontEndFloatModifier(Gameplay=1, Front End=0.7)
     Y=&FN_FrontEndFloatModifier(Gameplay=1, Front End=0.7)
     Z=&FloatFromCurve(
      FloatCurve=DataInterface() {propertyValues {Curve {keys [
          {interpMode=RCIM_Cubic, tangentMode=RCTM_User, time=0, value=0}
          {interpMode=RCIM_Cubic, tangentMode=RCTM_User, time=0.5, value=27}
          {tangentMode=RCTM_User, time=1, value=0}
         ]}}}
      Scale Curve=$Particles.MypositionSet
     )
    ))
   &ScaleMeshSize(Scale Factor=&MakeVector(
     X=1
     Y=1
     Z=&Lerp_Float(A=0.5, B=1.5, Alpha=&Power(a=$NPC.FortniteNPC.EventAudio_Amplitude, b=3))
    )) {module=ScaleMeshSize001}
  ]
  rendererClasses [&NiagaraMeshRendererProperties]
  renderers [{rendererClass=&NiagaraMeshRendererProperties, rendererIndex=0}]
 }
 Lightning_Arms_Body {
  emitterUpdateScript [
   &EmitterState(
    Inactive Response=::NewEnumerator1
    Life Cycle Mode=::NewEnumerator1
    Loop Duration=&RandomRangeFloat(Minimum=0.05, Maximum=1.25)
    Recalculate Duration Each Loop=true
    UseLoopDelay=false
   )
   &SpawnRate(
    Spawn Probability=&MakeCustomFloatFromBool(Bool=&Boolean_AndOperation(Bool1=$User.isBody, Bool2=$User.bUseReactivity), True Float=1, False Float=0)
    SpawnGroup=1
    SpawnRate=&Multiply_Float(
     A=1
     B=&Lerp_Float(
      A=0
      B=30
      Alpha=&FloatFromCurve(
       FloatCurve=DataInterface() {propertyValues {Curve {keys [
           <at 0.05 0 RCIM_Cubic>
           {interpMode=RCIM_Cubic, tangentMode=RCTM_User, time=0.12, value=1}
           {interpMode=RCIM_Cubic, tangentMode=RCTM_User, time=0.4, value=0}
           {arriveTangent=-0.1099, interpMode=RCIM_Cubic, leaveTangent=-0.1099, tangentMode=RCTM_User, time=0.9, value=0}
           {interpMode=RCIM_Cubic, tangentMode=RCTM_User, time=1, value=0.25}
          ]}}}
       CurveIndex=$System.Timeline.NormalizedTime
       Scale Curve=&FloatFromCurve(
        FloatCurve=DataInterface() {propertyValues {Curve {keys [
            {arriveTangent=4.1063, interpMode=RCIM_Cubic, leaveTangent=4.1063, tangentMode=RCTM_User, time=0.7, value=0}
            {arriveTangent=0.214, interpMode=RCIM_Cubic, leaveTangent=0.214, tangentMode=RCTM_User, time=0.84, value=0.3}
            {arriveTangent=8.1299, interpMode=RCIM_Cubic, leaveTangent=8.1299, tangentMode=RCTM_User, time=0.93, value=1}
           ]}}}
        CurveIndex=$NPC.FortniteNPC.EventAudio_Amplitude
       )
      )
     )
    )
    Use Spawn Probability=true
   ) {module=SpawnRate001}
   &SpawnRate(
    Spawn Probability=&MakeCustomFloatFromBool(Bool=&Boolean_AndOperation(Bool1=$User.isBody, Bool2=$User.bUseReactivity), True Float=1, False Float=0)
    SpawnRate=&Lerp_Float(
     A=0
     B=&FN_FrontEndFloatModifier(Gameplay=110, Front End=80)
     Alpha=&FloatFromCurve(
      FloatCurve=DataInterface() {propertyValues {Curve {keys [
          {arriveTangent=0.1665, interpMode=RCIM_Cubic, leaveTangent=0.1665, tangentMode=RCTM_User, time=0.18, value=0}
          {arriveTangent=-0.0039, interpMode=RCIM_Cubic, leaveTangent=-0.0039, tangentMode=RCTM_User, time=0.3, value=1}
          {arriveTangent=-2.4024, interpMode=RCIM_Cubic, leaveTangent=-2.4024, tangentMode=RCTM_User, time=0.85, value=0.1}
          {interpMode=RCIM_Cubic, tangentMode=RCTM_User, time=1, value=0.2}
         ]}}}
      CurveIndex=$System.Timeline.NormalizedTime
     )
    )
    Use Spawn Probability=true
   )
   /SetVariables(Emitter.Audio_Intensity:NiagaraFloat=&FloatFromCurve(
     FloatCurve=DataInterface() {propertyValues {Curve {keys [
         {arriveTangent=4.1063, interpMode=RCIM_Cubic, leaveTangent=4.1063, tangentMode=RCTM_User, time=0.7, value=0}
         {arriveTangent=0.214, interpMode=RCIM_Cubic, leaveTangent=0.214, tangentMode=RCTM_User, time=0.84, value=0.3}
         {arriveTangent=8.1299, interpMode=RCIM_Cubic, leaveTangent=8.1299, tangentMode=RCTM_User, time=0.97, value=1}
        ]}}}
     CurveIndex=$NPC.FortniteNPC.EventAudio_Amplitude
    )) {isSetParameters=true}
  ]
  particleSpawnScript [
   /SetVariables(Particles.IsMirrored:NiagaraBool=&RandomBool()) {isSetParameters=true}
   &InitializeParticle(
    Color=&MakeCustomLinearColorFromBool(Bool=$System.bUseRedderColors, True Color=<rgba 1.5 0.0225 0.0225 1>, False Color=<rgba 1.5 0.015 0.0397 1>)
    Lifetime=&MaskFloatBySpawnGroup(
     Masked Float=&Multiply_Float(
      A=&RandomRangeFloat(Minimum=0.3, Maximum=0.6)
      B=&Lerp_Float(
       A=0.6
       B=1
       Alpha=&FloatFromCurve(
        FloatCurve=DataInterface() {propertyValues {Curve {keys [
            {interpMode=RCIM_Cubic, tangentMode=RCTM_User, time=0.4, value=0}
            {arriveTangent=3.5471, interpMode=RCIM_Cubic, leaveTangent=3.5471, tangentMode=RCTM_User, time=1, value=1}
           ]}}}
        CurveIndex=$System.Timeline.NormalizedTime
       )
      )
     )
     Passthrough Float=&Multiply_Float(
      A=&RandomRangeFloat(Minimum=0.35, Maximum=0.45)
      B=&Lerp_Float(
       A=1
       B=1
       Alpha=&FloatFromCurve(
        FloatCurve=DataInterface() {propertyValues {Curve {keys [
            {arriveTangent=4.6454, interpMode=RCIM_Cubic, leaveTangent=4.6454, tangentMode=RCTM_User, time=0.85, value=0}
            {arriveTangent=6.6558, interpMode=RCIM_Cubic, leaveTangent=6.6558, tangentMode=RCTM_User, time=0.9, value=1}
           ]}}}
        CurveIndex=$NPC.FortniteNPC.EventAudio_Amplitude
       )
      )
     )
    )
    Mesh Scale=&MaskVectorBySpawnGroup(
     Masked Vector=&Multiply_VectorByFloat(
      Vector=&MakeCustomVectorFromBool(Bool=$Particles.IsMirrored, True Vector=<-1 -1 -1>, False Vector=<1 1 1>)
      Float=&RandomRangeFloat(Minimum=0.45, Maximum=0.95)
     )
     Passthrough Vector=&Multiply_VectorByFloat(
      Vector=&MakeCustomVectorFromBool(Bool=$Particles.IsMirrored, True Vector=<-1 1 1>, False Vector=<1 1 1>)
      Float=&RandomRangeFloat(Minimum=0.92, Maximum=1.2)
     )
    )
    Mesh Scale Mode=::NewEnumerator4
    Sprite Size Mode=::NewEnumerator3
    Uniform Sprite Size=40
    UsePositionOffset=false
    Write Material Random=true
   )
   /SetVariables(Particles.MeshIndex=&MaskIntBySpawnGroup(
     Masked Int=&RandomRangeInt(Minimum=0, Maximum=3)
     Passthrough Int=&RandomRangeInt(Minimum=4, Maximum=6)
     Spawn Group Mask=0
    )) {isSetParameters=true}
  ]
  particleUpdateScript [
   &ParticleState()
   &ScaleMeshSize(Scale Factor=&MaskVectorBySpawnGroup(
     Masked Vector=<1 1 1>
     Passthrough Vector=&Multiply_VectorByFloat(
      Vector=&VectorFromFloat(Value=&Lerp_Float(A=0.8, B=1, Alpha=&Power(a=$NPC.FortniteNPC.EventAudio_Amplitude, b=2)))
      Float=&FloatFromCurve(FloatCurve=DataInterface() {propertyValues {Curve {keys [
           {arriveTangent=0.0536, interpMode=RCIM_Cubic, leaveTangent=0.0536, tangentMode=RCTM_User, time=0.1, value=0.8}
           {arriveTangent=0.5097, interpMode=RCIM_Cubic, leaveTangent=0.5097, tangentMode=RCTM_User, time=1, value=1.1}
          ]}}})
     )
    ))
   /SetVariables(Particles.BoneLocation:NiagaraInt32=&MaskIntBySpawnGroup(
     Masked Int=&Add_Integer(
      A=&RandomRangeInt(
       Minimum=0
       Maximum=&MakeIntFromFloat(Float=&MakeCustomFloatFromBool(
         Bool=$User.bisHandHidden
         True Float=&FloatFromCurve(
          FloatCurve=DataInterface() {propertyValues {Curve {keys [<at 0 0 RCIM_Constant>,<at 0.4 1 RCIM_Constant>]}}}
          CurveIndex=$System.Timeline.NormalizedTime
         )
         False Float=&FloatFromCurve(
          FloatCurve=DataInterface() {propertyValues {Curve {keys [<at 0 0 RCIM_Constant>,<at 0.4 1 RCIM_Constant>,<at 0.6 2 RCIM_Constant>]}}}
          CurveIndex=$System.Timeline.NormalizedTime
         )
        ))
      )
      B=&MakeCustomIntFromBool(
       Bool=$Particles.IsMirrored
       True INT=&MakeCustomIntFromBool(Bool=$User.bisHandHidden, True INT=2, False INT=3)
       False INT=0
      )
     )
     Passthrough Int=7
     Spawn Group Mask=0
    )) {isSetParameters=true}
   &SkeletalMeshLocation(
    Bone / Socket Index=$Particles.BoneLocation
    Bone Sampling Mode=::NewEnumerator4
    MaskSpawnGroup=false
    Sampled Position Offset=&MaskVectorBySpawnGroup(
     Masked Vector=&AddVectorToPosition(
      Position=&RandomRangeVector(Minimum=<-3 -1 -1>, Maximum=<3 1 1>)
      Vector=&MakeCustomVectorFromBool(
       Bool=$User.bisHandHidden
       True Vector=&SelectVectorFromArray(
        Direct Array Index=$Particles.BoneLocation
        Vector Selection Array=DataInterface() {propertyValues {FloatData [<0 0 0>, <0 0 0>, <-7 0 0>, <7 0 0>, <0 0 0>, <0 0 0>]}}
       )
       False Vector=<1 0 0>
      )
     )
     Passthrough Vector=<0 12 0>
    )
    Skeletal Mesh=$System.SkeletalMesh
   )
   &DynamicMaterialParameters(
    Index 0 Param 1=&MaskFloatBySpawnGroup(
     Masked Float=&FloatFromCurve(FloatCurve=DataInterface() {propertyValues {Curve {keys [
          {interpMode=RCIM_Cubic, leaveTangent=3.6046, tangentMode=RCTM_Break, time=0, value=0.25}
          {interpMode=RCIM_Cubic, tangentMode=RCTM_User, time=0.2, value=0.5}
          {arriveTangent=0.0283, interpMode=RCIM_Cubic, tangentMode=RCTM_Break, time=1, value=1}
         ]}}})
     Passthrough Float=&FloatFromCurve(FloatCurve=DataInterface() {propertyValues {Curve {keys [
          {interpMode=RCIM_Cubic, leaveTangent=0.6442, tangentMode=RCTM_Break, time=0.1, value=0.25}
          {arriveTangent=0.0283, interpMode=RCIM_Cubic, tangentMode=RCTM_Break, time=1, value=1}
         ]}}})
    )
    Index 0 Param 2=&FloatFromCurve(FloatCurve=DataInterface() {propertyValues {Curve {keys [<at 0 0.5>, <at 0.2 1>, <at 0.4 1>, {arriveTangent=-4.0866, time=1, value=0}]}}})
    Index 0 Param 3=&FloatFromCurve(FloatCurve=DataInterface() {propertyValues {Curve {keys [<at 0 1 RCIM_Constant>,<at 0.4 0 RCIM_Constant>,<at 1 0>]}}}, Scale Curve=5)
    Index 0 Param 4=&FloatFromCurve(
     FloatCurve=DataInterface() {propertyValues {Curve {keys [<at 0 0 RCIM_Constant>,<at 0.2 0.5 RCIM_Constant>,<at 0.266 0.166 RCIM_Constant>,<at 0.333 1>]}}}
     Scale Curve=&MaskFloatBySpawnGroup(Masked Float=9, Passthrough Float=15)
    )
    Index 1 Param 1=&FloatFromCurve(FloatCurve=DataInterface() {propertyValues {Curve {keys [<at 0 0.5 RCIM_Constant>, <at 0.2 0 RCIM_Constant>, <at 0.4 0.75>, {tangentMode=RCTM_User, time=1, value=2}]}}})
    Index 1 Param 2=&FloatFromCurve(
     FloatCurve=DataInterface() {propertyValues {Curve {keys [
         {interpMode=RCIM_Cubic, leaveTangent=25.2051, tangentMode=RCTM_Break, time=0, value=0}
         <at 0.2 2 RCIM_Constant>
         <at 0.266 1>
         <at 0.85 1>
         <at 1 0>
        ]}}}
     Scale Curve=&Multiply_Float(
      A=&MakeCustomFloatFromBool(Bool=$Particles.IsHead, True Float=1.333, False Float=1.666)
      B=&MaskFloatBySpawnGroup(Masked Float=0.5, Passthrough Float=0.7)
     )
    )
    Index 1 Param 3=&FloatFromCurve(FloatCurve=DataInterface() {propertyValues {Curve {keys [<at 0 0.25>,<at 0.5 0.25>,<at 1 2>]}}})
    Index 1 Param 4=&Multiply_Float(
     A=&FloatFromCurve(FloatCurve=DataInterface() {propertyValues {Curve {keys [
          {interpMode=RCIM_Cubic, tangentMode=RCTM_User, time=0, value=0}
          {interpMode=RCIM_Cubic, tangentMode=RCTM_User, time=0.7, value=1}
          {interpMode=RCIM_Cubic, tangentMode=RCTM_User, time=1, value=0.6}
         ]}}})
     B=&Lerp_Float(
      A=0.9
      B=1.1
      Alpha=&FloatFromCurve(
       FloatCurve=DataInterface() {propertyValues {Curve {keys [{interpMode=RCIM_Cubic, tangentMode=RCTM_User, time=0.4, value=0}, {interpMode=RCIM_Cubic, tangentMode=RCTM_User, time=1, value=1}]}}}
       CurveIndex=$Emitter.Audio_Intensity
      )
     )
    )
    Module.Index0Param0Write=true
    Module.Index0Param1Write=true
    Module.Index0Param2Write=true
    Module.Index0Param3Write=true
    Module.Index1Param0Write=true
    Module.Index1Param1Write=true
    Module.Index1Param2Write=true
    Module.Index1Param3Write=true
    Write Parameter Index 1=true
   )
   /SetVariables(
    Particles.LightningGlowColor:LinearColor=&MultiplyLinearColor(
     Linear Color=$Particles.Color
     Scale Factor=&MakeFloat4FromVectorAndFloat("Vector (RGB)"=&VectorFromFloat(Value=0.5), "Float (Alpha)"=1)
    )
    Particles.LightningGlowDynamic0:Vector4f=&MakeVector4(
     X=&MakeFloatFromVector4(Vector 4=$Particles.DynamicMaterialParameter, Channel=::NewEnumerator0)
     Y=&MakeFloatFromVector4(Vector 4=$Particles.DynamicMaterialParameter, Channel=::NewEnumerator1)
     Z=0
     W=0
    )
    Particles.LightningGlowDynamic1:Vector4f=&MakeVector4(
     X=&MakeFloatFromVector4(Vector 4=$Particles.DynamicMaterialParameter1, Channel=::NewEnumerator0)
     Y=&Multiply_Float(
      A=&MakeFloatFromVector4(Vector 4=$Particles.DynamicMaterialParameter1, Channel=::NewEnumerator1)
      B=1.5
     )
     Z=0.25
     W=&MakeFloatFromVector4(Vector 4=$Particles.DynamicMaterialParameter1, Channel=::NewEnumerator3)
    )
    Particles.LightningGlowSoftColor:LinearColor=&MultiplyLinearColor(
     Linear Color=$Particles.Color
     Scale Factor=&MakeFloat4FromVectorAndFloat("Vector (RGB)"=&VectorFromFloat(Value=0.3), "Float (Alpha)"=1)
    )
    Particles.LightningGlowSoftDynamic1:Vector4f=&MakeVector4(
     X=&MakeFloatFromVector4(Vector 4=$Particles.DynamicMaterialParameter, Channel=::NewEnumerator0)
     Y=&Multiply_Float(
      A=&MakeFloatFromVector4(Vector 4=$Particles.DynamicMaterialParameter1, Channel=::NewEnumerator1)
      B=&MaskFloatBySpawnGroup(Masked Float=6, Passthrough Float=4)
     )
     Z=0
     W=&MakeFloatFromVector4(Vector 4=$Particles.DynamicMaterialParameter1, Channel=::NewEnumerator3)
    )
    Particles.LightningGlowSoftTopColor:LinearColor=&MakeCustomLinearColorFromBool(Bool=$System.bUseRedderColors, True Color=<rgba 0.9 0.0135 0.0135 0.2>, False Color=<rgba 0.9 0.09 0.1035 0.2>)
    Particles.LightningGlowSoftTopDynamic1:Vector4f=&MakeVector4(
     X=&MakeFloatFromVector4(Vector 4=$Particles.DynamicMaterialParameter, Channel=::NewEnumerator0)
     Y=&Multiply_Float(A=&MakeFloatFromVector4(Vector 4=$Particles.DynamicMaterialParameter1, Channel=::NewEnumerator1), B=3)
     Z=0
     W=&MakeFloatFromVector4(Vector 4=$Particles.DynamicMaterialParameter1, Channel=::NewEnumerator3)
    )
   ) {isSetParameters=true}
   &CameraOffset(Camera Offset Amount=&MaskFloatBySpawnGroup(Masked Float=3, Passthrough Float=4))
  ]
  rendererClasses [&NiagaraMeshRendererProperties]
  renderers [
   {rendererClass=&NiagaraMeshRendererProperties, rendererIndex=0}
   {rendererClass=&NiagaraMeshRendererProperties, rendererIndex=1}
   {rendererClass=&NiagaraMeshRendererProperties, rendererIndex=2}
   {rendererClass=&NiagaraMeshRendererProperties, rendererIndex=3}
  ]
 }
 Lightning_Large {
  emitterUpdateScript [
   &EmitterState(
    Inactive Response=::NewEnumerator1
    Life Cycle Mode=::NewEnumerator1
    Loop Duration=&RandomRangeFloat(Minimum=0.05, Maximum=1.25)
    Recalculate Duration Each Loop=true
    UseLoopDelay=false
   )
   &SpawnRate(
    Spawn Probability=&MakeCustomFloatFromBool(Bool=&Boolean_AndOperation(Bool1=$System.bIsBackpack, Bool2=$User.bUseReactivity), True Float=1, False Float=0)
    SpawnRate=&Multiply_Float(
     A=7
     B=&FloatFromCurve(
      FloatCurve=DataInterface() {propertyValues {Curve {keys [
          {arriveTangent=2.3272, interpMode=RCIM_Cubic, leaveTangent=2.3272, tangentMode=RCTM_User, time=0.4, value=0}
          {arriveTangent=0.214, interpMode=RCIM_Cubic, leaveTangent=0.214, tangentMode=RCTM_User, time=0.85, value=0.45}
          {arriveTangent=8.1299, interpMode=RCIM_Cubic, leaveTangent=8.1299, tangentMode=RCTM_User, time=0.9, value=1}
         ]}}}
      CurveIndex=$System.MusicReactive
     )
    )
    Use Spawn Probability=true
   )
   /SetVariables(Emitter.Audio_Intensity:NiagaraFloat=&FloatFromCurve(
     FloatCurve=DataInterface() {propertyValues {Curve {keys [
         {arriveTangent=4.1063, interpMode=RCIM_Cubic, leaveTangent=4.1063, tangentMode=RCTM_User, time=0.7, value=0}
         {arriveTangent=0.214, interpMode=RCIM_Cubic, leaveTangent=0.214, tangentMode=RCTM_User, time=0.84, value=0.3}
         {arriveTangent=8.1299, interpMode=RCIM_Cubic, leaveTangent=8.1299, tangentMode=RCTM_User, time=0.97, value=1}
        ]}}}
     CurveIndex=$NPC.FortniteNPC.EventAudio_Amplitude
    )) {isSetParameters=true, module=SetVariables_#5}
  ]
  particleSpawnScript [
   &PartitionParticles(Number Of Partitions=6)
   &InitializeParticle(
    Color=&MakeCustomLinearColorFromBool(Bool=$System.bUseRedderColors, True Color=<rgba 3 0.045 0.045 1>, False Color=<rgba 3 0 0.1 1>)
    Lifetime=&Multiply_Float(
     A=&RandomRangeFloat(Minimum=0.5, Maximum=0.7)
     B=&Lerp_Float(
      A=0.9
      B=1
      Alpha=&FloatFromCurve(
       FloatCurve=DataInterface() {propertyValues {Curve {keys [
           {arriveTangent=4.6454, interpMode=RCIM_Cubic, leaveTangent=4.6454, tangentMode=RCTM_User, time=0.85, value=0}
           {arriveTangent=6.6558, interpMode=RCIM_Cubic, leaveTangent=6.6558, tangentMode=RCTM_User, time=0.9, value=1}
          ]}}}
       CurveIndex=$NPC.FortniteNPC.EventAudio_Amplitude
      )
     )
    )
    Mesh Scale=&Multiply_VectorByFloat(
     Vector=&RandomRangeVector(Minimum=<0.8 0.8 0.8>, Maximum=<1 1 1>)
     Float=&RandomRangeFloat(Minimum=0.91, Maximum=1.1)
    )
    Mesh Scale Mode=::NewEnumerator4
    Position Mode=::NewEnumerator0
    Sprite Size Mode=::NewEnumerator3
    Uniform Sprite Size=70
    Write Material Random=true
   )
   /SetVariables(Particles.IsMirrored:NiagaraBool=&RandomBool()) {isSetParameters=true}
   /SetVariables(Particles.MeshIndex=$Particles.PartitionParticles.Partition) {isSetParameters=true}
   /SetVariables(Particles.Position=&Multiply_Vector(
     A=&MakeCustomVectorFromBool(Bool=$Particles.IsMirrored, True Vector=<-1 1 1>, False Vector=<1 1 1>)
     B=&AddVectorToPosition(
      Position=&Multiply_VectorByFloat(
       Vector=&RandomRangeVector(Minimum=<-6 -1 -3>, Maximum=<6 1 3>)
       Float=&SelectFloatFromArray(
        Direct Array Index=$Particles.PartitionParticles.Partition
        Float Selection Array=DataInterface() {propertyValues {FloatData [0.8,1,0.9,0.5,0.5,1]}}
       )
      )
      Vector=&SelectVectorFromArray(
       Direct Array Index=$Particles.PartitionParticles.Partition
       Vector Selection Array=DataInterface() {propertyValues {FloatData [<10 -2 8>, <-15 -2 10>, <6 -2 10>, <-18 -2 9>, <-12 -4 -11>, <12 -2 -8>]}}
      )
     )
    )) {isSetParameters=true}
   /SetVariables(Particles.Scale=&MakeCustomVectorFromBool(Bool=$Particles.IsMirrored, True Vector=<-1 1 1>, False Vector=<1 1 1>)) {isSetParameters=true}
  ]
  particleUpdateScript [
   &ParticleState()
   &DynamicMaterialParameters(
    Index 0 Param 1=&FloatFromCurve(FloatCurve=DataInterface() {propertyValues {Curve {keys [
         {interpMode=RCIM_Cubic, leaveTangent=1.3752, tangentMode=RCTM_Break, time=0, value=0.25}
         {arriveTangent=0.0283, interpMode=RCIM_Cubic, tangentMode=RCTM_Break, time=1, value=1}
        ]}}})
    Index 0 Param 2=&FloatFromCurve(FloatCurve=DataInterface() {propertyValues {Curve {keys [<at 0 0.5>, <at 0.2 1>, <at 0.4 1>, {arriveTangent=-4.0866, time=1, value=0}]}}})
    Index 0 Param 3=&FloatFromCurve(FloatCurve=DataInterface() {propertyValues {Curve {keys [<at 0 1 RCIM_Constant>,<at 0.4 0 RCIM_Constant>,<at 1 0>]}}}, Scale Curve=10)
    Index 0 Param 4=&FloatFromCurve(
     FloatCurve=DataInterface() {propertyValues {Curve {keys [<at 0 0 RCIM_Constant>,<at 0.2 0.5 RCIM_Constant>,<at 0.266 0.166 RCIM_Constant>,<at 0.333 1>]}}}
     Scale Curve=15
    )
    Index 1 Param 1=&FloatFromCurve(FloatCurve=DataInterface() {propertyValues {Curve {keys [<at 0 0.5 RCIM_Constant>, <at 0.2 0 RCIM_Constant>, <at 0.4 0.75>, {tangentMode=RCTM_User, time=1, value=2}]}}})
    Index 1 Param 2=&FloatFromCurve(
     FloatCurve=DataInterface() {propertyValues {Curve {keys [
         {interpMode=RCIM_Cubic, leaveTangent=25.2051, tangentMode=RCTM_Break, time=0, value=0}
         <at 0.2 2 RCIM_Constant>
         <at 0.266 1>
         <at 0.85 1>
         <at 1 0>
        ]}}}
     Scale Curve=&Multiply_Float(A=&MakeCustomFloatFromBool(Bool=$Particles.IsHead, True Float=1.333, False Float=1.666), B=0.5)
    )
    Index 1 Param 3=&FloatFromCurve(FloatCurve=DataInterface() {propertyValues {Curve {keys [<at 0 0.25>,<at 0.5 0.25>,<at 1 2>]}}})
    Index 1 Param 4=&Multiply_Float(
     A=&FloatFromCurve(FloatCurve=DataInterface() {propertyValues {Curve {keys [{interpMode=RCIM_Cubic, tangentMode=RCTM_User, time=0, value=0}, {interpMode=RCIM_Cubic, tangentMode=RCTM_User, time=0.7, value=1}]}}})
     B=&Lerp_Float(
      A=0.9
      B=1.1
      Alpha=&FloatFromCurve(
       FloatCurve=DataInterface() {propertyValues {Curve {keys [{interpMode=RCIM_Cubic, tangentMode=RCTM_User, time=0.4, value=0}, {interpMode=RCIM_Cubic, tangentMode=RCTM_User, time=1, value=1}]}}}
       CurveIndex=$Emitter.Audio_Intensity
      )
     )
    )
    Module.Index0Param0Write=true
    Module.Index0Param1Write=true
    Module.Index0Param2Write=true
    Module.Index0Param3Write=true
    Module.Index1Param0Write=true
    Module.Index1Param1Write=true
    Module.Index1Param2Write=true
    Module.Index1Param3Write=true
    Write Parameter Index 1=true
   )
   &CameraOffset(Camera Offset Amount=1)
   /SetVariables(
    Particles.LightningGlowColor:LinearColor=&MultiplyLinearColor(
     Linear Color=$Particles.Color
     Scale Factor=&MakeFloat4FromVectorAndFloat("Vector (RGB)"=&VectorFromFloat(Value=0.5), "Float (Alpha)"=1)
    )
    Particles.LightningGlowDynamic0:Vector4f=&MakeVector4(
     X=&MakeFloatFromVector4(Vector 4=$Particles.DynamicMaterialParameter, Channel=::NewEnumerator0)
     Y=&MakeFloatFromVector4(Vector 4=$Particles.DynamicMaterialParameter, Channel=::NewEnumerator1)
     Z=0
     W=0
    )
    Particles.LightningGlowDynamic1:Vector4f=&MakeVector4(
     X=&MakeFloatFromVector4(Vector 4=$Particles.DynamicMaterialParameter1, Channel=::NewEnumerator0)
     Y=&Multiply_Float(
      A=&MakeFloatFromVector4(Vector 4=$Particles.DynamicMaterialParameter1, Channel=::NewEnumerator1)
      B=1.5
     )
     Z=0.25
     W=&MakeFloatFromVector4(Vector 4=$Particles.DynamicMaterialParameter1, Channel=::NewEnumerator3)
    )
    Particles.LightningGlowSoftColor:LinearColor=&MultiplyLinearColor(
     Linear Color=$Particles.Color
     Scale Factor=&MakeFloat4FromVectorAndFloat("Vector (RGB)"=&VectorFromFloat(Value=0.3), "Float (Alpha)"=1)
    )
    Particles.LightningGlowSoftDynamic1:Vector4f=&MakeVector4(
     X=&MakeFloatFromVector4(Vector 4=$Particles.DynamicMaterialParameter, Channel=::NewEnumerator0)
     Y=&Multiply_Float(A=&MakeFloatFromVector4(Vector 4=$Particles.DynamicMaterialParameter1, Channel=::NewEnumerator1), B=6)
     Z=0
     W=&MakeFloatFromVector4(Vector 4=$Particles.DynamicMaterialParameter1, Channel=::NewEnumerator3)
    )
    Particles.LightningGlowSoftTopColor:LinearColor=<rgba 0.7 0.07 0.0805 0.2>
    Particles.LightningGlowSoftTopDynamic1:Vector4f=&MakeVector4(
     X=&MakeFloatFromVector4(Vector 4=$Particles.DynamicMaterialParameter, Channel=::NewEnumerator0)
     Y=&Multiply_Float(A=&MakeFloatFromVector4(Vector 4=$Particles.DynamicMaterialParameter1, Channel=::NewEnumerator1), B=3)
     Z=0
     W=&MakeFloatFromVector4(Vector 4=$Particles.DynamicMaterialParameter1, Channel=::NewEnumerator3)
    )
   ) {isSetParameters=true, module=SetVariables_#6}
   /SetVariables(Particles.CameraOffsetGlow=10) {isSetParameters=true, module=SetVariables_#7}
   /SetVariables(Particles.ColorGlow:LinearColor=&ColorFromCurve(ColorCurve=DataInterface() {propertyValues {
       AlphaCurve {keys [<at 0 0>,<at 0.1669 0.5 RCIM_Cubic>,<at 0.8184 0 RCIM_Cubic>]}
       BlueCurve {keys [<at 0 0.0265 RCIM_Cubic>]}
       GreenCurve {keys [<at 0 0.01 RCIM_Cubic>]}
       RedCurve {keys [<at 0 1 RCIM_Cubic>]}
      }})) {isSetParameters=true, module=SetVariables_#8}
  ]
  rendererClasses [&NiagaraMeshRendererProperties]
  renderers [
   {rendererClass=&NiagaraMeshRendererProperties, rendererIndex=0}
   {rendererClass=&NiagaraMeshRendererProperties, rendererIndex=1}
   {rendererClass=&NiagaraMeshRendererProperties, rendererIndex=2}
   {rendererClass=&NiagaraMeshRendererProperties, rendererIndex=3}
  ]
 }
 Lightning_Speakers {
  emitterUpdateScript [
   &EmitterState(
    Inactive Response=::NewEnumerator1
    Life Cycle Mode=::NewEnumerator1
    Loop Duration=&RandomRangeFloat(Minimum=0.05, Maximum=1.25)
    Recalculate Duration Each Loop=true
    UseLoopDelay=false
   )
   &SpawnRate(
    Spawn Probability=&MakeCustomFloatFromBool(Bool=&Boolean_AndOperation(Bool1=$System.bIsBackpack, Bool2=$User.bUseReactivity), True Float=1, False Float=0)
    SpawnRate=&Multiply_Float(
     A=20
     B=&FN_FrontEndFloatModifier(
      Gameplay=&FloatFromCurve(
       FloatCurve=DataInterface() {propertyValues {Curve {keys [
           {arriveTangent=0.214, interpMode=RCIM_Cubic, leaveTangent=0.214, tangentMode=RCTM_User, time=0.9, value=0}
           {arriveTangent=8.1299, interpMode=RCIM_Cubic, leaveTangent=8.1299, tangentMode=RCTM_User, time=0.93, value=1}
          ]}}}
       CurveIndex=$System.MusicReactive
      )
      Front End=&FloatFromCurve(
       FloatCurve=DataInterface() {propertyValues {Curve {keys [
           {arriveTangent=0.214, interpMode=RCIM_Cubic, leaveTangent=0.214, tangentMode=RCTM_User, time=0.4, value=0}
           {arriveTangent=8.1299, interpMode=RCIM_Cubic, leaveTangent=8.1299, tangentMode=RCTM_User, time=0.93, value=1}
          ]}}}
       CurveIndex=$System.MusicReactive
      )
     )
    )
    Use Spawn Probability=true
   )
   /SetVariables(Emitter.Audio_Intensity:NiagaraFloat=&FloatFromCurve(
     FloatCurve=DataInterface() {propertyValues {Curve {keys [
         {arriveTangent=4.1063, interpMode=RCIM_Cubic, leaveTangent=4.1063, tangentMode=RCTM_User, time=0.7, value=0}
         {arriveTangent=0.214, interpMode=RCIM_Cubic, leaveTangent=0.214, tangentMode=RCTM_User, time=0.84, value=0.3}
         {arriveTangent=8.1299, interpMode=RCIM_Cubic, leaveTangent=8.1299, tangentMode=RCTM_User, time=0.97, value=1}
        ]}}}
     CurveIndex=$NPC.FortniteNPC.EventAudio_Amplitude
    )) {isSetParameters=true, module=SetVariables_#5}
  ]
  particleSpawnScript [
   /SetVariables(Particles.IsMirrored:NiagaraBool=&RandomBool()) {isSetParameters=true}
   &InitializeParticle(
    Color=&MakeCustomLinearColorFromBool(Bool=$System.bUseRedderColors, True Color=<rgba 6 0.12 0.12 1>, False Color=<rgba 6 0 0.2 1>)
    Lifetime=&Multiply_Float(
     A=&RandomRangeFloat(Minimum=0.3, Maximum=0.4)
     B=&Lerp_Float(
      A=1.3
      B=0.8
      Alpha=&FloatFromCurve(
       FloatCurve=DataInterface() {propertyValues {Curve {keys [
           {arriveTangent=4.6454, interpMode=RCIM_Cubic, leaveTangent=4.6454, tangentMode=RCTM_User, time=0.85, value=0}
           {arriveTangent=6.6558, interpMode=RCIM_Cubic, leaveTangent=6.6558, tangentMode=RCTM_User, time=0.9, value=1}
          ]}}}
       CurveIndex=$NPC.FortniteNPC.EventAudio_Amplitude
      )
     )
    )
    Mesh Scale=&Multiply_VectorByFloat(Vector=<1 1 1>, Float=&RandomRangeFloat(Minimum=1, Maximum=1.15))
    Mesh Scale Mode=::NewEnumerator4
    Position Offset=&SelectVectorFromArray(
     Array Sampling Mode=::NewEnumerator0
     Direct Array Index=?
     Vector Selection Array=DataInterface() {propertyValues {FloatData [<-11 1 -0.7>, <11 1 -0.7>]}}
     Random Seed=0
    )
    Sprite Size Mode=::NewEnumerator3
    Uniform Sprite Size=70
    UsePositionOffset=true
    Write Material Random=true
   )
   /SetVariables(Particles.MeshIndex=&RandomRangeInt(Minimum=0, Maximum=2)) {isSetParameters=true}
   &InitialMeshOrientation(
    Mesh Orientation Mode=::NewEnumerator1
    Rotation=&RandomRangeVector(Minimum=<0 -1 0>, Maximum=<0 1 0>)
    Use Rotation Vector=true
   )
  ]
  particleUpdateScript [
   &ParticleState()
   &DynamicMaterialParameters(
    Index 0 Param 1=&FloatFromCurve(FloatCurve=DataInterface() {propertyValues {Curve {keys [
         {interpMode=RCIM_Cubic, leaveTangent=3.6046, tangentMode=RCTM_Break, time=0, value=0.25}
         {interpMode=RCIM_Cubic, tangentMode=RCTM_User, time=0.2, value=0.5}
         {arriveTangent=0.0283, interpMode=RCIM_Cubic, tangentMode=RCTM_Break, time=1, value=1}
        ]}}})
    Index 0 Param 2=&FloatFromCurve(FloatCurve=DataInterface() {propertyValues {Curve {keys [<at 0 0.5>, <at 0.2 1>, <at 0.4 1>, {arriveTangent=-4.0866, time=1, value=0}]}}})
    Index 0 Param 3=&FloatFromCurve(FloatCurve=DataInterface() {propertyValues {Curve {keys [<at 0 1 RCIM_Constant>,<at 0.4 0 RCIM_Constant>,<at 1 0>]}}}, Scale Curve=40)
    Index 0 Param 4=&FloatFromCurve(
     FloatCurve=DataInterface() {propertyValues {Curve {keys [<at 0 0 RCIM_Constant>,<at 0.2 0.5 RCIM_Constant>,<at 0.266 0.166 RCIM_Constant>,<at 0.333 1>]}}}
     Scale Curve=15
    )
    Index 1 Param 1=&FloatFromCurve(FloatCurve=DataInterface() {propertyValues {Curve {keys [<at 0 0.5 RCIM_Constant>, <at 0.2 0 RCIM_Constant>, <at 0.4 0.75>, {tangentMode=RCTM_User, time=1, value=2}]}}})
    Index 1 Param 2=&FloatFromCurve(
     FloatCurve=DataInterface() {propertyValues {Curve {keys [
         {interpMode=RCIM_Cubic, leaveTangent=25.2051, tangentMode=RCTM_Break, time=0, value=0}
         <at 0.2 2 RCIM_Constant>
         <at 0.266 1>
         <at 0.85 1>
         <at 1 0>
        ]}}}
     Scale Curve=&Multiply_Float(A=&MakeCustomFloatFromBool(Bool=$Particles.IsHead, True Float=1.333, False Float=1.666), B=0.4)
    )
    Index 1 Param 3=&FloatFromCurve(FloatCurve=DataInterface() {propertyValues {Curve {keys [<at 0 0.25>,<at 0.5 0.25>,<at 1 2>]}}})
    Index 1 Param 4=&Multiply_Float(
     A=&FloatFromCurve(
      FloatCurve=DataInterface() {propertyValues {Curve {keys [{interpMode=RCIM_Cubic, tangentMode=RCTM_User, time=0, value=0}, {interpMode=RCIM_Cubic, tangentMode=RCTM_User, time=0.7, value=1}]}}}
      Scale Curve=0.1
     )
     B=&Lerp_Float(
      A=0.9
      B=1.1
      Alpha=&FloatFromCurve(
       FloatCurve=DataInterface() {propertyValues {Curve {keys [{interpMode=RCIM_Cubic, tangentMode=RCTM_User, time=0.4, value=0}, {interpMode=RCIM_Cubic, tangentMode=RCTM_User, time=1, value=1}]}}}
       CurveIndex=$Emitter.Audio_Intensity
      )
     )
    )
    Module.Index0Param0Write=true
    Module.Index0Param1Write=true
    Module.Index0Param2Write=true
    Module.Index0Param3Write=true
    Module.Index1Param0Write=true
    Module.Index1Param1Write=true
    Module.Index1Param2Write=true
    Module.Index1Param3Write=true
    Write Parameter Index 1=true
   )
   &CameraOffset(Camera Offset Amount=1)
   /SetVariables(
    Particles.LightningGlowColor:LinearColor=&MultiplyLinearColor(
     Linear Color=$Particles.Color
     Scale Factor=&MakeFloat4FromVectorAndFloat("Vector (RGB)"=&VectorFromFloat(Value=0.5), "Float (Alpha)"=1)
    )
    Particles.LightningGlowDynamic0:Vector4f=&MakeVector4(
     X=&MakeFloatFromVector4(Vector 4=$Particles.DynamicMaterialParameter, Channel=::NewEnumerator0)
     Y=&MakeFloatFromVector4(Vector 4=$Particles.DynamicMaterialParameter, Channel=::NewEnumerator1)
     Z=0
     W=0
    )
    Particles.LightningGlowDynamic1:Vector4f=&MakeVector4(
     X=&MakeFloatFromVector4(Vector 4=$Particles.DynamicMaterialParameter1, Channel=::NewEnumerator0)
     Y=&Multiply_Float(
      A=&MakeFloatFromVector4(Vector 4=$Particles.DynamicMaterialParameter1, Channel=::NewEnumerator1)
      B=1.5
     )
     Z=0.25
     W=&MakeFloatFromVector4(Vector 4=$Particles.DynamicMaterialParameter1, Channel=::NewEnumerator3)
    )
    Particles.LightningGlowSoftColor:LinearColor=&MultiplyLinearColor(
     Linear Color=$Particles.Color
     Scale Factor=&MakeFloat4FromVectorAndFloat("Vector (RGB)"=&VectorFromFloat(Value=0.3), "Float (Alpha)"=1)
    )
    Particles.LightningGlowSoftDynamic1:Vector4f=&MakeVector4(
     X=&MakeFloatFromVector4(Vector 4=$Particles.DynamicMaterialParameter, Channel=::NewEnumerator0)
     Y=&Multiply_Float(A=&MakeFloatFromVector4(Vector 4=$Particles.DynamicMaterialParameter1, Channel=::NewEnumerator1), B=4)
     Z=0
     W=&MakeFloatFromVector4(Vector 4=$Particles.DynamicMaterialParameter1, Channel=::NewEnumerator3)
    )
    Particles.LightningGlowSoftTopColor:LinearColor=<rgba 0.5 0 0.0083 0.2>
    Particles.LightningGlowSoftTopDynamic1:Vector4f=&MakeVector4(
     X=&MakeFloatFromVector4(Vector 4=$Particles.DynamicMaterialParameter, Channel=::NewEnumerator0)
     Y=&Multiply_Float(A=&MakeFloatFromVector4(Vector 4=$Particles.DynamicMaterialParameter1, Channel=::NewEnumerator1), B=3)
     Z=0
     W=&MakeFloatFromVector4(Vector 4=$Particles.DynamicMaterialParameter1, Channel=::NewEnumerator3)
    )
   ) {isSetParameters=true, module=SetVariables_#6}
   /SetVariables(Particles.CameraOffsetGlow=10) {isSetParameters=true, module=SetVariables_#7}
   /SetVariables(Particles.ColorGlow:LinearColor=&ColorFromCurve(ColorCurve=DataInterface() {propertyValues {
       AlphaCurve {keys [<at 0 0>,<at 0.1669 0.5 RCIM_Cubic>,<at 0.8184 0 RCIM_Cubic>]}
       BlueCurve {keys [<at 0 0.0265 RCIM_Cubic>]}
       GreenCurve {keys [<at 0 0.01 RCIM_Cubic>]}
       RedCurve {keys [<at 0 1 RCIM_Cubic>]}
      }})) {isSetParameters=true, module=SetVariables_#8}
  ]
  rendererClasses [&NiagaraMeshRendererProperties]
  renderers [
   {rendererClass=&NiagaraMeshRendererProperties, rendererIndex=0}
   {rendererClass=&NiagaraMeshRendererProperties, rendererIndex=1}
   {rendererClass=&NiagaraMeshRendererProperties, rendererIndex=2}
   {rendererClass=&NiagaraMeshRendererProperties, rendererIndex=3}
  ]
 }
 SplineThickenMeshes {
  bEnabled=false
  emitterSpawnScript [/SetVariables(
    Emitter.BIsBody:NiagaraBool=$User.isBody
    Emitter.bIsBackpack:NiagaraBool=&Boolean_NotOperation(Bool=$User.isBody)
   ) {isSetParameters=true}]
  emitterUpdateScript [
   &EmitterState(
    Inactive Response=::NewEnumerator1
    Life Cycle Mode=::NewEnumerator1
    Loop Duration=&RandomRangeFloat(Minimum=0.133, Maximum=0.5)
    Recalculate Duration Each Loop=true
    UseLoopDelay=false
   )
   &SpawnRate(Spawn Probability=0.333, SpawnRate=6, Use Spawn Probability=true)
   &SpawnBurst_Instantaneous(Spawn Probability=0.5, Use Spawn Probability=true, UseLoopCountLimit=false)
  ]
  particleSpawnScript [
   &InitializeParticle(
    Color=<rgba 1 0 0.0382 1>
    Lifetime Max=0.7332
    Lifetime Min=0.433
    Lifetime Mode=::NewEnumerator1
    Mesh Scale Mode=::NewEnumerator3
    UsePositionOffset=true
    Write Material Random=true
   )
   /SetVariables(Particles.MeshIndex=&ModuloInt(
     A=&ReturnParticleIDAttributes()
     B=&self:ScratchDynamicInput(Backpack:NiagaraInt32=4, Body:NiagaraInt32=12, bool:NiagaraBool=$Emitter.BIsBody)
    )) {isSetParameters=true}
   /SetVariables(Particles.MeshIndex=11) {enabled=false, isSetParameters=true}
  ]
  particleUpdateScript [
   &ParticleState()
   &ScaleMeshSize(Scale Factor=&VectorFromFloat(Value=&FloatFromCurve(FloatCurve=DataInterface() {propertyValues {Curve {keys [<at 0 1>,<at 0.9 1>]}}})))
   &DynamicMaterialParameters(
    Index 0 Param 1=&FloatFromCurve(FloatCurve=DataInterface() {propertyValues {Curve {keys [
         {interpMode=RCIM_Cubic, leaveTangent=4.2199, tangentMode=RCTM_Break, time=0, value=0}
         <at 0.25 0.5 RCIM_Constant>
         {interpMode=RCIM_Cubic, leaveTangent=2.2211, tangentMode=RCTM_Break, time=0.5, value=0.666}
         {arriveTangent=0.0283, interpMode=RCIM_Cubic, tangentMode=RCTM_Break, time=1, value=1}
        ]}}})
    Index 0 Param 2=&FloatFromCurve(FloatCurve=DataInterface() {propertyValues {Curve {keys [<at 0 0.5>, <at 0.2 1>, <at 0.4 1>, {arriveTangent=-4.0866, time=1, value=0}]}}})
    Index 0 Param 3=&FloatFromCurve(FloatCurve=DataInterface() {propertyValues {Curve {keys [<at 0 1 RCIM_Constant>,<at 0.4 1>,<at 0.5 0 RCIM_Constant>,<at 1 0>]}}})
    Index 0 Param 4=&FloatFromCurve(
     FloatCurve=DataInterface() {propertyValues {Curve {keys [<at 0 0 RCIM_Constant>,<at 0.25 1 RCIM_Constant>,<at 0.333 0.166 RCIM_Constant>,<at 0.4 1>]}}}
     Scale Curve=33
    )
    Index 1 Param 1=&FloatFromCurve(FloatCurve=DataInterface() {propertyValues {Curve {keys [<at 0 0.5 RCIM_Constant>, <at 0.2 0.2 RCIM_Constant>, <at 0.4 0.75>, {tangentMode=RCTM_User, time=1, value=2}]}}})
    Index 1 Param 2=&FloatFromCurve(
     FloatCurve=DataInterface() {propertyValues {Curve {keys [
         {interpMode=RCIM_Cubic, leaveTangent=25.2051, tangentMode=RCTM_Break, time=0, value=0}
         <at 0.2 2 RCIM_Constant>
         <at 0.266 1>
         <at 0.333 1>
         <at 1 0>
        ]}}}
     Scale Curve=0.5
    )
    Index 1 Param 3=0.5
    Module.Index0Param0Write=true
    Module.Index0Param1Write=true
    Module.Index0Param2Write=true
    Module.Index0Param3Write=true
    Module.Index1Param0Write=true
    Module.Index1Param1Write=true
    Module.Index1Param2Write=true
    Module.Index1Param3Write=false
    Write Parameter Index 1=true
   )
   /SetVariables(
    Particles.LightColor=&MultiplyLinearColor(
     Linear Color=<rgba 9 0.0735 0 1>
     Scale Factor=&MakeVector4_FromFloat(Float=&FloatFromCurve(FloatCurve=DataInterface() {propertyValues {Curve {keys [
           {arriveTangent=9.0725, interpMode=RCIM_Cubic, leaveTangent=9.0725, tangentMode=RCTM_User, time=0, value=0}
           {arriveTangent=-2.5249, interpMode=RCIM_Cubic, leaveTangent=-2.5249, tangentMode=RCTM_User, time=0.1408, value=0.9847}
           {arriveTangent=-2.2063, interpMode=RCIM_Cubic, leaveTangent=-2.2063, tangentMode=RCTM_User, time=0.5, value=0}
          ]}}}))
    )
    Particles.LightPosition:NiagaraPosition=&self:ScratchDynamicInput_01(
     bool:NiagaraBool=$User.isBody
     VectorBackpack:Vector3f=&SelectVectorFromArray(
      Direct Array Index=$Particles.MeshIndex
      Vector Selection Array=DataInterface() {propertyValues {FloatData [<-10 -10 10>, <20 -10 10>, <-8 -15 -5>, <20 -10 -8>, <-25 -10 -15>, <6 -10 -10>]}}
     )
     VectorBody:Vector3f=&SelectVectorFromArray(
      Direct Array Index=$Particles.MeshIndex
      Vector Selection Array=DataInterface() {propertyValues {FloatData [
         <-15 10 100>
         <20 20 120>
         <-15 10 120>
         <-7 15 100>
         <-20 -10 100>
         <-20 -10 150>
         <14 10 110>
         <-25 -10 100>
         <20 10 105>
         <20 20 145>
         <0 20 100>
        ]}}
     )
    )
    Particles.LightRadius=&FloatFromCurve(FloatCurve=DataInterface() {propertyValues {Curve {keys [<at 0 1 RCIM_Constant>,<at 0.5 0>]}}}, Scale Curve=70)
    Particles.LightningGlowColor:LinearColor=&MultiplyLinearColor(
     Linear Color=$Particles.Color
     Scale Factor=&MakeFloat4FromVectorAndFloat("Vector (RGB)"=&VectorFromFloat(Value=2), "Float (Alpha)"=0.1)
    )
    Particles.LightningGlowDynamic0:Vector4f=&MakeVector4(
     X=&MakeFloatFromVector4(Vector 4=$Particles.DynamicMaterialParameter, Channel=::NewEnumerator0)
     Y=&MakeFloatFromVector4(Vector 4=$Particles.DynamicMaterialParameter, Channel=::NewEnumerator1)
     Z=0
     W=0
    )
    Particles.LightningGlowDynamic1:Vector4f=&MakeVector4(
     X=&MakeFloatFromVector4(Vector 4=$Particles.DynamicMaterialParameter1, Channel=::NewEnumerator0)
     Y=&Multiply_Float(A=&MakeFloatFromVector4(Vector 4=$Particles.DynamicMaterialParameter1, Channel=::NewEnumerator1), B=4)
     Z=&MakeFloatFromVector4(Vector 4=$Particles.DynamicMaterialParameter1, Channel=::NewEnumerator2)
     W=0
    )
    Particles.LightningGlowSoftColor:LinearColor=&MultiplyLinearColor(
     Linear Color=$Particles.Color
     Scale Factor=&MakeFloat4FromVectorAndFloat("Vector (RGB)"=&VectorFromFloat(Value=0.166), "Float (Alpha)"=1)
    )
    Particles.LightningGlowSoftDynamic1:Vector4f=&MakeVector4(
     X=&MakeFloatFromVector4(Vector 4=$Particles.DynamicMaterialParameter, Channel=::NewEnumerator0)
     Y=&Multiply_Float(A=&MakeFloatFromVector4(Vector 4=$Particles.DynamicMaterialParameter1, Channel=::NewEnumerator1), B=6)
     Z=&MakeFloatFromVector4(Vector 4=$Particles.DynamicMaterialParameter1, Channel=::NewEnumerator2)
     W=0
    )
    Particles.LightningGlowSoftTopColor:LinearColor=<rgba 0.25 0.5 0.4866 0.2>
    Particles.LightningGlowSoftTopDynamic1:Vector4f=&MakeVector4(
     X=&MakeFloatFromVector4(Vector 4=$Particles.DynamicMaterialParameter, Channel=::NewEnumerator0)
     Y=&Multiply_Float(A=&MakeFloatFromVector4(Vector 4=$Particles.DynamicMaterialParameter1, Channel=::NewEnumerator1), B=4)
     Z=&MakeFloatFromVector4(Vector 4=$Particles.DynamicMaterialParameter1, Channel=::NewEnumerator2)
     W=0
    )
    Particles.SphereGlowColor:LinearColor=<rgba 0 0.25 1 1>
    Particles.SphereGlowScale:Vector3f=<1 1 1>
   ) {isSetParameters=true}
  ]
  rendererClasses [&NiagaraMeshRendererProperties, &NiagaraLightRendererProperties]
  renderers [
   {rendererClass=&NiagaraMeshRendererProperties, rendererIndex=0}
   {rendererClass=&NiagaraMeshRendererProperties, rendererIndex=1}
   {rendererClass=&NiagaraMeshRendererProperties, rendererIndex=2}
   {rendererClass=&NiagaraMeshRendererProperties, rendererIndex=3}
   {rendererClass=&NiagaraLightRendererProperties, rendererIndex=4}
   {rendererClass=&NiagaraMeshRendererProperties, rendererIndex=5}
   {rendererClass=&NiagaraMeshRendererProperties, rendererIndex=6}
   {rendererClass=&NiagaraMeshRendererProperties, rendererIndex=7}
  ]
 }
}
summary {
 systemName=NS_WarmShadeWagon_Splines
 emitters [
  {
   bEnabled=false
   emitterName=SplineThickenMeshes
   rendererClasses [&NiagaraMeshRendererProperties, &NiagaraLightRendererProperties]
  }
  {emitterName=Backpack_SpeakerRings, rendererClasses [&NiagaraMeshRendererProperties]}
  {emitterName=Lightning_Speakers, rendererClasses [&NiagaraMeshRendererProperties]}
  {emitterName=Backpack_TopGlow, rendererClasses [&NiagaraMeshRendererProperties]}
  {emitterName=Lightning_Large, rendererClasses [&NiagaraMeshRendererProperties]}
  {emitterName=Lightning_Arms_Body, rendererClasses [&NiagaraMeshRendererProperties]}
 ]
 userVariables [
  User.bIsFrontEnd {type=NiagaraBool}
  User.bIsFrontEndPreview {type=NiagaraBool}
  User.bisHandHidden {type=NiagaraBool}
  User.bisInVehicle {type=NiagaraBool}
  User.bUseReactivity {type=NiagaraBool}
  User.isBody {type=NiagaraBool}
  User.ReactiveDebug {type=NiagaraBool}
 ]
}
